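; Procedural erosion Blueprint that stamps a material-driven deformation and a paint layer weight onto the landscape through a Landscape Texture Patch.
; Construction creates height and weight instances of M_ErosionRT with strength, contrast and blur, assigning the Texture only to height, plus an editor-only debug plane.
; UpdateDeformation configures the patch, draws both materials into its height and weight render targets, then requests a landscape update.
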(event EventBeginPlay)

(event Collision|EventActorBeginOverlap
  (params (OtherActor :type Actor%20Object%20Reference)))

(event EventTick
  (params (DeltaSeconds :type Float%20%28single-precision%29)))

(event Custom|UpdateDeformation
  (bind _landscapetexturepatch (Variables|Default|GetLandscapeTexturePatch))
  (LandscapePatch|SetUnscaledCoverage (Variables|Default|GetLandscapeTexturePatch) (Math|Vector2D|MakeVector2D (Variables|Default|GetSize) (Variables|Default|GetSize)))
  (LandscapePatch|SetResolution (Variables|Default|GetLandscapeTexturePatch) "(X=512.000000,Y=512.000000)")
  (LandscapePatch|SetHeightSourceMode (Variables|Default|GetLandscapeTexturePatch) "TextureBackedRenderTarget")
  (LandscapePatch|SetIsEnabled (Variables|Default|GetLandscapeTexturePatch) (Variables|Default|GetEnabled))
  (LandscapePatch|SetHeightRenderTargetFormat _landscapetexturepatch "RTF_R32f")
  (bind _returnvalue (LandscapePatch|GetHeightRenderTarget _landscapetexturepatch true))
  (Rendering|ClearRenderTarget2D _returnvalue)
  (Rendering|DrawMaterialtoRenderTarget _returnvalue (Variables|Default|GetMID_Height))
  (LandscapePatch|CreateWeightPatch (Variables|Default|GetLandscapeTexturePatch) (Variables|Default|GetWeightLayerName) "TextureBackedRenderTarget" "None")
  (bind _returnvalue_2 (LandscapePatch|GetWeightPatchRenderTarget (Variables|Default|GetLandscapeTexturePatch) (Variables|Default|GetWeightLayerName) true))
  (Rendering|ClearRenderTarget2D _returnvalue_2)
  (Rendering|DrawMaterialtoRenderTarget _returnvalue_2 (Variables|Default|GetMID_Weight))
  (LandscapePatch|RequestLandscapeUpdate (Variables|Default|GetLandscapeTexturePatch) false true))

(fn ConstructionScript
  (bind _editoronlyvismesh (Variables|Default|GetEditorOnlyVisMesh))
  (LandscapePatch|SetIsEnabled (Variables|Default|GetLandscapeTexturePatch) (Variables|Default|GetEnabled))
  (bind _returnvalue (Rendering|Material|CreateDynamicMaterialInstance "/Game/ArtTools/Procedural/Erosion/M_ErosionRT.M_ErosionRT" "None" "None" :!class /Script/Engine.KismetMaterialLibrary))
  (Variables|Default|SetMID_Height _returnvalue)
  (bind _returnvalue_2 (Rendering|Material|CreateDynamicMaterialInstance "/Game/ArtTools/Procedural/Erosion/M_ErosionRT.M_ErosionRT" "None" "None" :!class /Script/Engine.KismetMaterialLibrary))
  (Variables|Default|SetMID_Weight _returnvalue_2)
  (bind _returnvalue_3 (Rendering|Material|CreateDynamicMaterialInstance "/Game/ArtTools/Procedural/Erosion/M_VisualizeErosionTexture.M_VisualizeErosionTexture" "None" "None" :!class /Script/Engine.KismetMaterialLibrary))
  (Variables|Default|SetVisualizeMaterial _returnvalue_3)
  (Rendering|Material|SetScalarParameterValue (Variables|Default|GetMID_Height) "Height" (Variables|Default|GetDeformation) :!class /Script/Engine.MaterialInstanceDynamic)
  (Rendering|Material|SetScalarParameterValue (Variables|Default|GetMID_Weight) "Height" (Variables|Default|GetWeightOpacity) :!class /Script/Engine.MaterialInstanceDynamic)
  (Rendering|Material|SetTextureParameterValue (Variables|Default|GetMID_Height) "Texture" (Variables|Default|GetTexture))
  (Rendering|Material|SetScalarParameterValue (Variables|Default|GetMID_Height) "Contrast" (Math|Float|Max(Float) (Variables|Default|GetDeformationContrast)) :!class /Script/Engine.MaterialInstanceDynamic)
  (Rendering|Material|SetScalarParameterValue (Variables|Default|GetMID_Weight) "Contrast" (Math|Float|Max(Float) (Variables|Default|GetWeightContrast)) :!class /Script/Engine.MaterialInstanceDynamic)
  (Rendering|Material|SetScalarParameterValue (Variables|Default|GetMID_Weight) "Blur" (Variables|Default|GetWeightBlur) :!class /Script/Engine.MaterialInstanceDynamic)
  (Rendering|Material|SetScalarParameterValue (Variables|Default|GetMID_Height) "Blur" (Variables|Default|GetDeformationBlur) :!class /Script/Engine.MaterialInstanceDynamic)
  (if (Editor|InEditor)
    (Rendering|Material|SetMaterial _editoronlyvismesh 0 (Variables|Default|GetVisualizeMaterial))
    (Transformation|SetRelativeScale3D _editoronlyvismesh (Math|Conversions|ToVector(Float) (Variables|Default|GetSize)))
    (Rendering|SetVisibility _editoronlyvismesh (Variables|Default|GetShowDebugPlane) false)))
