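# Post-process material in the procedural erosion art tools that turns a grayscale erosion mask into a height value, apparently for a render target.
# The Texture parameter feeds a SpiralBlur with high step counts, with Blur setting the distance, clamped just above zero.
# The red channel then passes through CheapContrast and is multiplied by Height, with negative emissive values allowed.
# Two extra samples of the same Texture parameter sit unconnected.
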Material /Game/ArtTools/Procedural/Erosion/M_ErosionRT
graph {
 properties {
  bAllowNegativeEmissiveColor=true
  blendMode=BLEND_Translucent
  blendableOutputAlpha=true
  materialDomain=MD_PostProcess
  pixelDepthOffsetMode=PDOM_Legacy
  shadingModel=MSM_Unlit
  usedShadingModels=Unlit
 }
 nodes [
  n1 = TextureObjectParameter() {
   parameterName=Texture
   samplerType=SAMPLERTYPE_LinearGrayscale
   texture=/Game/ArtTools/Procedural/Erosion/ErosionTextures/T_DecalMask_Erosion04
  }
  n2 = ScalarParameter() {parameterName=Blur}
  n3 = Constant() {r=32}
  n4 = MaterialFunctionCall!(
   TextureObject=@n1
   UV
   Distance=Multiply(A=Max(A=@n2) {constB=0.001}) {constB=0.05}
   Distance Steps=@n3
   Radial Steps=@n3
   Radial Offset
   Kernel Power
  ) {materialFunction=/Engine/Functions/Engine_MaterialFunctions01/Texturing/SpiralBlur-Texture}
  Multiply(
   A=MaterialFunctionCall(
    In=MaterialFunctionCall(Float3=@n4.Result) {materialFunction=&BreakOutFloat3Components}.R
    Contrast=ScalarParameter() {parameterName=Contrast}
   ) {materialFunction=&CheapContrast}.Result
   B=ScalarParameter() {defaultValue=1, parameterName=Height}
  ) {constB=-300}
  TextureSampleParameter2D() {
   parameterName=Texture
   samplerType=SAMPLERTYPE_LinearGrayscale
   texture=&T_DecalMask_Erosion04
  }
  TextureSampleParameter2D() {
   parameterName=Texture
   samplerType=SAMPLERTYPE_LinearGrayscale
   texture=&T_DecalMask_Erosion04
  }
 ]
}
