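# Translucent unlit material that, going by its name, visualizes procedural erosion mask textures from the art tools.
# A grayscale Texture parameter, defaulting to an erosion decal mask, has its red channel run through CheapContrast with an adjustable Contrast parameter.
# A separate branch builds an inverted constant grey and passes it through EyeAdaptationInverse to counter auto-exposure.
# The graph does not show which material outputs the two branches drive.
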Material /Game/ArtTools/Procedural/Erosion/M_VisualizeErosionTexture
graph {
 properties {blendMode=BLEND_Translucent, pixelDepthOffsetMode=PDOM_Legacy, shadingModel=MSM_Unlit, usedShadingModels=Unlit}
 nodes [
  n1 = Constant() {r=0.8}
  EyeAdaptationInverse(LightValueInput=OneMinus(None=MaterialFunctionCall(X=@n1, Y=@n1, Z=@n1) {materialFunction=&MakeFloat3}.Result))
  n2 = TextureSampleParameter2D() {
   parameterName=Texture
   samplerType=SAMPLERTYPE_LinearGrayscale
   texture=/Game/ArtTools/Procedural/Erosion/ErosionTextures/T_DecalMask_Erosion04
  }
  MaterialFunctionCall(In=@n2.R, Contrast=ScalarParameter() {parameterName=Contrast}) {materialFunction=&CheapContrast}
 ]
}
